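; Mounted turret vehicle Blueprint that drives the gun's overheat visuals and audio.
; Overheat changes run StartOverHeatChanges directly or through MultiCastOverheatValueChanged, depending on the PerfHotfix scalable float, with effects gated to cosmetic-capable contexts.
; Heating and cooling feed the heat ratio to Niagara and a smoothed audio parameter, while overheating calls GunOverheat, spawns cooling FX, plays a sound and prints debug text.
; Pawn entry zeroes the Niagara heat, and exit also stops FX and looping audio.
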(event EventBeginPlay
  (|CallParentFunction) ; This node is disabled and will not be called.\nDrag off pins to build functionality.
)

(event Collision|EventActorBeginOverlap
  (params (OtherActor :type Actor%20Object%20Reference)))

(event EventTick
  (params (DeltaSeconds :type Float%20%28single-precision%29)))

(event Custom|StartOverHeatChanges
  (params (CurrentValue :type Float%20%28double-precision%29) (MaxValue :type Float%20%28double-precision%29) (State :type EFortWeaponOverheatState%20Enum) (IsInCooldown :type Boolean))
  (bind _returnvalue (/ CurrentValue MaxValue))
  (bind _ns_lightturret_overheat (Variables|Default|GetNS_LightTurret_Overheat))
  (bind _ns_lightturret_overheat_2 (Variables|Default|GetNS_LightTurret_Overheat))
  ;[1 0.5445 0.3833] Handle Overheat
  (Variables|Default|SetIs%20in%20Cooldown IsInCooldown)
  (Networking|CanExecuteCosmeticEvents
    (:True
      (switch Utilities|FlowControl|Switch|SwitchonEFortWeaponOverheatState State
        (:None
          (CallFunction|StopOverheatAudio))
        (:Heating
          (Utilities|IsValid _ns_lightturret_overheat_2
            (:Is%20Valid
              (if (not (Components|Activation|IsActive _ns_lightturret_overheat))
                (Components|Activation|Activate _ns_lightturret_overheat false))
              (Niagara|SetNiagaraVariable(Float) _ns_lightturret_overheat_2 "User.OverheatAmount" _returnvalue)
              (CallFunction|OverheatBuildUpAudio _returnvalue (Variables|Default|GetIs%20in%20Cooldown)))))
        (:Cooling
          (Utilities|IsValid (Variables|Default|GetNS_LightTurret_Overheat)
            (:Is%20Valid
              (Niagara|SetNiagaraVariable(Float) _ns_lightturret_overheat_2 "User.OverheatAmount" _returnvalue)
              (CallFunction|OverheatBuildUpAudio _returnvalue (Variables|Default|GetIs%20in%20Cooldown))
              (if (<= _returnvalue 0)
                (Components|Activation|Deactivate (Variables|Default|GetNS_LightTurret_Overheat))))))
        (:Overheated
          (Class|FortVehicleInterface|CallGunOverheat self (Variables|Default|GetFortVehicleAimingWeaponComp))
          (Development|PrintString "Overheat FX play" true true "(R=1.000000,G=0.022914,B=0.000000,A=1.000000)" 1e+01)
          (Effects|SpawnFXSystemAttached :FXSystemAsset "/ResolveTurret/Items/Environmental/SevenMountedTurret/FX/NS_SMT_Turret_Overheat_Cooling.NS_SMT_Turret_Overheat_Cooling" :AttachToComponent (Variables|Vehicle|GetSkeletalMesh) :AttachPointName "root" :Scale "1.000000,1.000000,1.000000" :LocationType "SnapToTarget" :bAutoDestroy true :PoolingMethod "None" :bAutoActivate true)
          (Audio|Components|Audio|Play (Variables|Default|GetAudioOnOverheat) 0.0)
          (CallFunction|StopOverheatAudio))))))

(event Eventoverheatchanged
  (params (CurrentValue :type Float%20%28single-precision%29) (MaxValue :type Float%20%28single-precision%29) (State :type EFortWeaponOverheatState%20Enum) (bIsInCooldown :type Boolean))
  ;[1 0.5445 0.3833] Handle Overheat
  (if (>= (ScalableFloat|GetValueatLevel (Variables|Perf|GetPerfHotfix)) 1.0)
    (CallFunction|StartOverHeatChanges CurrentValue MaxValue State bIsInCooldown)
    (else
      (CallFunction|MultiCastOverheatValueChanged CurrentValue MaxValue State bIsInCooldown))))

(event Custom|MultiCastOverheatValueChanged
  (params (CurrentValue :type Float%20%28double-precision%29) (MaxValue :type Float%20%28double-precision%29) (State :type EFortWeaponOverheatState%20Enum) (IsInCooldown :type Boolean))
  ;[1 0.5445 0.3833] Handle Overheat
  (CallFunction|StartOverHeatChanges CurrentValue MaxValue State IsInCooldown))

(event EventOnPawnEnterVehicle
  (params (PlayerPawn :type Fort%20Player%20Pawn%20Object%20Reference) (SeatIdx :type Integer))
  (bind _ns_lightturret_overheat (Variables|Default|GetNS_LightTurret_Overheat))
  ;[1 0.5445 0.3833] Handle Overheat
  ;;[1 1 1] Clean and reset overheat FX while the pawn entering and exiting
  (|CallParentFunction PlayerPawn SeatIdx)
  (Utilities|IsValid _ns_lightturret_overheat
    (:Is%20Valid
      (Niagara|SetNiagaraVariable(Float) _ns_lightturret_overheat "User.OverheatAmount" 0.0))))

(event EventEndPlay
  (params (EndPlayReason :type EEndPlayReason%20Enum :default "Destroyed")))

(event EventOnPawnExitVehicle
  (params (PlayerPawn :type Fort%20Player%20Pawn%20Object%20Reference) (ExitSocketName :type Name :default "None"))
  (bind _ns_lightturret_overheat (Variables|Default|GetNS_LightTurret_Overheat))
  ;[1 0.5445 0.3833] Handle Overheat
  ;;[1 1 1] Clean and reset overheat FX while the pawn entering and exiting
  (|CallParentFunction PlayerPawn ExitSocketName)
  (Utilities|IsValid _ns_lightturret_overheat
    (:Is%20Valid
      (Niagara|SetNiagaraVariable(Float) _ns_lightturret_overheat "User.OverheatAmount" 0.0)
      (Components|Activation|Deactivate _ns_lightturret_overheat)
      (CallFunction|StopOverheatAudio))))

(fn OverheatBuildUpAudio
  (params (OverheatPercent :type Float%20%28double-precision%29) (CoolDown :type Boolean))
  (bind _audiooverheat (Variables|Default|GetAudioOverheat))
  (Utilities|FlowControl|Sequence
    (:then_0
      (if (not (Audio|Components|Audio|IsPlaying _audiooverheat))
        (Audio|Components|Audio|FadeIn _audiooverheat 0.25 1.0 0.0 "Linear" :!class /Script/Engine.AudioComponent)))
    (:then_1
      (bind _output_get (Variables|Default|SetOverheatAudioParam (Math|Interpolation|FInterpTo (Variables|Default|GetOverheatAudioParam) (Math|Float|SelectFloat 0.0 OverheatPercent CoolDown) (Utilities|Time|GetWorldDeltaSeconds) 3.0)))
      (Audio|Parameter|SetFloatParameter (Variables|Default|GetAudioOverheat) "Overheat" _output_get :!class /Script/AudioExtensions.AudioParameterControllerInterface))))

(fn StopOverheatAudio
  (bind _audiooverheat (Variables|Default|GetAudioOverheat)
    (:then
      (if (Audio|Components|Audio|IsPlaying _audiooverheat)
        (Audio|Components|Audio|FadeOut _audiooverheat 1.5 0.0 "Linear" :!class /Script/Engine.AudioComponent)
        (Variables|Default|SetOverheatAudioParam 0.0)))
    (:else)))

(fn ConstructionScript
  (|CallParentFunction))
